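# Physics asset for the HIS_2 hair part of the female medium Elastic body, defining capsule bodies and joints for hair physics.
# A head capsule is jointed to a small hair root body whose swing and twist are all free.
# Hair segments hang off that root with tight angular limits, slightly looser on B_l and C_l.
# Only A_b_l chains further, from A_a_l, rather than attaching to the root.
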PhysicsAsset /BRCosmetics/Characters/Player/Female/Medium/Bodies/F_MED_Elastic/Meshes/Parts/HIS_Parts/F_MED_Elastic_HIS_2_Physics
bodies {
 head [{shapeType=Capsule, length=4, radius=6.75, center=<3.3006 2.1142 0>, rotation=<rot 90 0 0>}]
 his2_hair [{shapeType=Capsule, length=1, radius=1}]
 his2_hair_A [{shapeType=Capsule, length=1.75, radius=1, center=<1.2285 0 0>, rotation=<rot 90 0 0>}]
 his2_hair_A_a_l [{shapeType=Capsule, length=1.5, radius=1, center=<1.5435 0 0>, rotation=<rot 90 0 0>}]
 his2_hair_A_b_l [{shapeType=Capsule, length=2.5, radius=1, center=<1.9416 0 0>, rotation=<rot 90 0 0>}]
 his2_hair_B [{shapeType=Capsule, length=1.75, radius=1, center=<1.2285 0 0>, rotation=<rot 90 0 0>}]
 his2_hair_B_l [{shapeType=Capsule, length=2.25, radius=1, center=<1.2565 0 0.5632>, rotation=<rot 65.3988 -180 -180>}]
 his2_hair_C_l [{shapeType=Capsule, length=2.5, radius=1, center=<0.8526 0 1.0003>, rotation=<rot -40.4402 0 0>}]
}
constraints [
 {bone1Name=his2_hair_B_l, bone2Name=his2_hair, swing1LimitDegrees=20, swing2LimitDegrees=20, twistLimitDegrees=20}
 {bone1Name=his2_hair_C_l, bone2Name=his2_hair, swing1LimitDegrees=20, swing2LimitDegrees=20, twistLimitDegrees=20}
 {bone1Name=his2_hair_A, bone2Name=his2_hair, swing1LimitDegrees=10, swing2LimitDegrees=10, twistLimitDegrees=10}
 {bone1Name=his2_hair_B, bone2Name=his2_hair, swing1LimitDegrees=10, swing2LimitDegrees=10, twistLimitDegrees=10}
 {bone1Name=his2_hair_A_a_l, bone2Name=his2_hair, swing1LimitDegrees=10, swing2LimitDegrees=10, twistLimitDegrees=5}
 {bone1Name=his2_hair_A_b_l, bone2Name=his2_hair_A_a_l, swing1LimitDegrees=10, swing2LimitDegrees=10, twistLimitDegrees=5}
 {
  bone1Name=his2_hair
  bone2Name=head
  swing1LimitDegrees=45
  swing1Motion=Free
  swing2LimitDegrees=45
  swing2Motion=Free
  twistLimitDegrees=45
  twistMotion=Free
 }
]
