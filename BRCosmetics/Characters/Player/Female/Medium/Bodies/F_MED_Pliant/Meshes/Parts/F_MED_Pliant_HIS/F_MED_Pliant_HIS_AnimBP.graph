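# Post-copy AnimBlueprint for the HIS mesh part of the female medium Pliant cosmetic body, adding secondary physics motion.
# It copies the pose from the attached parent mesh, then applies an additive Y/Z scale to the dyn_collar bone.
# A RigidBody node then simulates in spine_03 space with damped component motion and overridden gravity, borrowing the physics asset of the Elastic HIS part.
# The event graph holds no logic.
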AnimBlueprint /BRCosmetics/Characters/Player/Female/Medium/Bodies/F_MED_Pliant/Meshes/Parts/F_MED_Pliant_HIS/F_MED_Pliant_HIS_AnimBP
target_skeleton=&/F_MED_Pliant_HIS_Skeleton
graphs [
 AnimGraph {nodes [
   n1 = "Transform (Modify) Bone - Bone: dyn_collar"(
    ComponentPose=Local To Component(LocalPose=Copy Pose From Mesh() {bUseAttachedParent=true})
    Alpha=1
    Translation="0,0,0"
    Rotation="0,0,0"
    Scale="1,1.3,1.3"
   ) {scaleMode=BMM_Additive, alphaBoolBlend {blendOption=Linear}, boneToModify {boneName=dyn_collar}, scale {y=1.3, z=1.3}}
   n2 = RigidBody(ComponentPose=@n1) {
    bClampLinearTranslationLimitToRefPose=true
    bOverrideWorldGravity=true
    overridePhysicsAsset=/BRCosmetics/Characters/Player/Female/Medium/Bodies/F_MED_Elastic/Meshes/Parts/HIS_Parts/F_MED_Elastic_HIS_2_Physics
    simulationSpace=BaseBoneSpace
    alphaBoolBlend {blendOption=Linear}
    baseBoneRef {boneName=spine_03}
    componentLinearAccScale=<0.4 0.4 0.4>
    componentLinearVelScale=<0.4 0.4 0.4>
    overrideWorldGravity {z=-980}
   }
   Output Pose(Result=Component To Local(ComponentPose=@n2)) {layerGroup=None}
  ]}
 EventGraph {nodes [Event BlueprintUpdateAnimation()]}
]
variables []
